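# HUD widget that shows the off-hand weapon's magazine and reserve ammo, ammo icon and a NO AMMO warning for dual-wield weapons.
# View target changes rebind it through the pawn's weapon-equipped event to FortOffhandComponent equip events, then to the off-hand weapon's ammo-changed event.
# Delta gates swap delegates as targets change, and the widget hides without an off-hand ranged weapon.
# Ammo updates show infinity symbols above a threshold and dim empty counts.
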WidgetBlueprint /BattleRoyaleDualWeaponUI/Widgets/WBP_DualWeapon_OffHandAmmo
graphs [
 BindToOffhandWeaponEvents {nodes [
   n1 = BindToOffhandWeaponEvents!() -> (then, NewOffhandComponent:Fort Offhand Component Object Reference) {is=title K2Node_FunctionEntry}
   > n2 = Delta Gate Wildcard!(
    execute:Exec
    Variable:"Fort Offhand Component Object Reference (by ref)"=Get BoundOffhandComponent() -> (BoundOffhandComponent:Fort Offhand Component Object Reference)
    New Value:Fort Offhand Component Object Reference=@n1
   ) -> (Success:Exec, Updated Variable:"Fort Offhand Component Object Reference (by ref)", Previous Variable:Fort Offhand Component Object Reference) {is=title K2Node_MacroInstance}
   > n3 = Sequence()
   > Is Valid(InputObject=@n2.Previous Variable)
   n4 = Create Event()
   > Unbind Event from On Equip Offhand Weapon Delegate(self:Fort Offhand Component Object Reference=@n2.Previous Variable, Delegate=@n4) {is=title K2Node_RemoveDelegate}
   n5 = Create Event()
   > Unbind Event from On Unequip Offhand Weapon Delegate(self:Fort Offhand Component Object Reference=@n2.Previous Variable, Delegate=@n5) {is=title K2Node_RemoveDelegate}
   > OnUnequipOffhandWeapon()
   Is Valid(exec=@n3.then_1, InputObject=Get BoundOffhandComponent())
   > Bind Event to On Equip Offhand Weapon Delegate(self:Fort Offhand Component Object Reference=Get BoundOffhandComponent(), Delegate=@n4) {is=title K2Node_AddDelegate}
   > Bind Event to On Unequip Offhand Weapon Delegate(self:Fort Offhand Component Object Reference=Get BoundOffhandComponent(), Delegate=@n5) {is=title K2Node_AddDelegate}
   > n6 = GetCurrentEquippedOffhandWeapon!(execute, self:Fort Offhand Component Object Reference=Get BoundOffhandComponent()) -> (then, ReturnValue:Fort Weapon Object Reference)
   > OnEquipOffhandWeapon!(execute, self, NewWeapon:Fort Weapon Object Reference=@n6) -> (then)
  ]}
 BindToPlayerEvents {nodes [
   n1 = BindToPlayerEvents!() -> (then, NewPawn:Fort Player Pawn Object Reference) {is=title K2Node_FunctionEntry}
   > n2 = Delta Gate Wildcard(
    Variable:"Fort Player Pawn Object Reference (by ref)"=Get BoundPawn() -> (BoundPawn:Fort Player Pawn Object Reference)
    New Value:Fort Player Pawn Object Reference=@n1
   ) -> (Updated Variable:"Fort Player Pawn Object Reference (by ref)", Previous Variable:Fort Player Pawn Object Reference)
   > n3 = Is Valid(InputObject=@n2.Previous Variable)
   n4 = Create Event()
   > Unbind Event from On Weapon Equipped Delegate(self:Fort Pawn Object Reference=@n2.Previous Variable, Delegate=@n4) {is=title K2Node_RemoveDelegate}
   > n5 = BindToOffhandWeaponEvents!(execute, self, NewOffhandComponent:Fort Offhand Component Object Reference) -> (then) {is=title K2Node_CallFunction}
   n6 = Get!(execute:Exec=[@n5.then, @n3.Is Not Valid]) -> (then:Exec, else:Exec, BoundPawn:Fort Player Pawn Object Reference)
   > Bind Event to On Weapon Equipped Delegate(self:Fort Pawn Object Reference=@n6, Delegate=@n4) {is=title K2Node_AddDelegate}
   > OnWeaponEquipped!(
    execute
    self
    NewWeapon:Fort Weapon Object Reference=GetCurrentWeapon!(self:Fort Pawn Object Reference=@n6) -> (ReturnValue:Fort Weapon Object Reference)
    PrevWeapon:Fort Weapon Object Reference
   ) -> (then)
  ]}
 EventGraph {nodes [
   n1 = OnAmmoChanged!() -> (OutputDelegate, then, LocalCount:int, LocalRemaining:int) {is=title K2Node_CustomEvent}
   > n2 = Branch(Condition="integer < integer"(A=@n1.LocalCount, B=Get InfiniteAmmo() -> (InfiniteAmmo:int=0)))
   > n3 = SetNoAmmoTextVisibility!(execute, self, CurrentAmmo:int=@n1.LocalCount) -> (then, NoAmmoTextIsVisible:bool)
   > n4 = Branch(Condition=@n3)
   Is Valid(exec=@n4.else, InputObject=Get OffhandWeapon() -> (OffhandWeapon:Fort Weapon Ranged Object Reference))
   > Set Text(Text="To Text (Integer)"(Value=@n1.LocalCount), self=Get Text_CurrentAmmo() -> (Text_CurrentAmmo:Common Text Object Reference))
   n5 = Get OffhandWeapon()
   n6 = int - int(A=GetLocalRemainingAmmo!(self:Fort Weapon Object Reference=@n5) -> (ReturnValue:int=0), B=@n1.LocalCount)
   n7 = GetRemainingAmmo!(self:Fort Weapon Object Reference=@n5) -> (ReturnValue:int=0)
   n8 = Get InfiniteAmmo() -> (InfiniteAmmo=0)
   n9 = Get Text_TotalAmmo() -> (Text_TotalAmmo:Common Text Object Reference)
   > Set Text(
    Text=SelectText(A="To Text (Integer)"(Value=@n6), B=GetInfinitySymbolText!() -> (ReturnValue:"Text (by ref)"), bPickA="integer < integer"(A=@n7, B=@n8))
    self=@n9
   )
   > SetVisibility(self=[Get Image_AmmoSplit() -> (Image_AmmoSplit:Image Object Reference), @n9], InVisibility=SelfHitTestInvisible)
   n10 = Get OffhandWeapon()
   n11 = GetLocalMagazineAmmoCount!(self:Fort Weapon Object Reference=@n10) -> (ReturnValue:int=0)
   > UpdateTextOpacity!(execute, self, TextWidget:Common Text Object Reference=Get Text_CurrentAmmo(), CurrentValue:int=@n11) -> (then)
   > UpdateTextOpacity(TextWidget=Get Text_TotalAmmo(), CurrentValue=int - int(A=GetLocalRemainingAmmo(self=@n10) -> (ReturnValue=0), B=@n11))
   Set Text(execute=@n2.else, Text=GetInfinitySymbolText(), self=Get Text_CurrentAmmo())
   > SetVisibility(self=[Get Text_TotalAmmo(), Get Image_AmmoSplit()], InVisibility=Collapsed)
   Event Construct()
   > SetVisibility(InVisibility=Hidden)
   n12 = GetOwningPlayer()
   > n13 = Cast To FortPlayerController(Object=@n12)
   > n14 = Set OwningFortPlayerController(OwningFortPlayerController:Fort Player Controller Object Reference=@n13) -> (Output_Get:Fort Player Controller Object Reference)
   > Bind Event to On View Target Changed(self:Fort Player Controller Object Reference=@n14, Delegate=Create Event()) {is=title K2Node_AddDelegate}
   > OnViewTargetChanged!(
    execute
    self
    PlayerController:Fort Player Controller Object Reference
    OldViewTarget:Actor Object Reference
    NewViewTarget:Actor Object Reference=GetViewTarget!(self:Controller Object Reference=@n12) -> (ReturnValue:Actor Object Reference)
   ) -> (then)
   Event Destruct()
   > BindToPlayerEvents!(execute, self, NewPawn:Fort Player Pawn Object Reference) -> (then) {is=title K2Node_CallFunction}
   n15 = Get OwningFortPlayerController() -> (OwningFortPlayerController:Fort Player Controller Object Reference)
   > Is Valid(InputObject=@n15)
   > Unbind Event from On View Target Changed(self:Fort Player Controller Object Reference=@n15, Delegate=Create Event()) {is=title K2Node_RemoveDelegate}
   > Set OwningFortPlayerController()
  ]}
 OnEquipOffhandWeapon {nodes [
   n1 = OnEquipOffhandWeapon!() -> (then, NewWeapon:Fort Weapon Object Reference) {is=title K2Node_FunctionEntry}
   > n2 = Cast To FortWeaponRanged!(execute, Object=@n1) -> (then, CastFailed, AsFort Weapon Ranged:Fort Weapon Ranged Object Reference) {is=title K2Node_DynamicCast}
   > n3 = Set NewOffhandWeapon(NewOffhandWeapon:Fort Weapon Ranged Object Reference=@n2) -> (Output_Get:Fort Weapon Ranged Object Reference)
   > Branch(Condition="Not Equal (Object)"(A:Object Reference=@n3, B:Object Reference=Get OffhandWeapon()) -> (ReturnValue:bool=~) {is=title K2Node_PromotableOperator})
   > OnUnequipOffhandWeapon()
   n4 = Get NewOffhandWeapon() -> (NewOffhandWeapon:Fort Weapon Ranged Object Reference)
   > n5 = Set OffhandWeapon(OffhandWeapon:Fort Weapon Ranged Object Reference=@n4) -> (Output_Get:Fort Weapon Ranged Object Reference)
   > Is Valid(InputObject=@n5)
   > Bind Event to On Local Ammo Changed(self:Fort Weapon Object Reference=@n4, Delegate=Create Event()) {is=title K2Node_AddDelegate}
   > OnAmmoChanged!(execute, self, LocalCount:int=GetMagazineAmmoCount!(self:Fort Weapon Object Reference=@n4) -> (ReturnValue:int=0), LocalRemaining:int=0) -> (then) {is=title K2Node_CallFunction}
   > SetVisibility(InVisibility=SelfHitTestInvisible)
   > n6 = Set AmmoItemDefinition(AmmoItemDefinition:Fort World Item Definition Object Reference=GetWeaponDataAmmoWorldItemDefinition!(self:Fort Weapon Object Reference=Get OffhandWeapon()) -> (ReturnValue:Fort World Item Definition Object Reference)) -> (Output_Get:Fort World Item Definition Object Reference)
   > Is Valid(InputObject=@n6)
   > SetBrushFromSoftTexture!(
    execute
    self:Image Object Reference=Get Image_AmmoType() -> (Image_AmmoType:Image Object Reference)
    SoftTexture:Texture 2D Soft Object Reference=GetSmallPreviewImage!(ItemDefinition:Item Definition Base Object Reference=@n6) -> (ReturnValue:Texture 2D Soft Object Reference)
    bMatchSize:bool=false
   ) -> (then)
   OnUnequipOffhandWeapon(execute=@n2.CastFailed)
  ]}
 OnUnequipOffhandWeapon {nodes [
   OnUnequipOffhandWeapon() {is=title K2Node_FunctionEntry}
   n1 = Get OffhandWeapon()
   > Is Valid(InputObject=@n1)
   > Unbind Event from On Local Ammo Changed(self:Fort Weapon Object Reference=@n1, Delegate=Create Event()) {is=title K2Node_RemoveDelegate}
   > Set OffhandWeapon()
   > SetVisibility(InVisibility=Hidden)
  ]}
 OnViewTargetChanged {nodes [
   n1 = OnViewTargetChanged!() -> (then, PlayerController:Fort Player Controller Object Reference, OldViewTarget:Actor Object Reference, NewViewTarget:Actor Object Reference) {is=title K2Node_FunctionEntry}
   > BindToPlayerEvents(NewPawn=Cast To FortPlayerPawn!(Object=@n1.NewViewTarget) -> (AsFort Player Pawn, bSuccess:bool=~))
  ]}
 OnWeaponEquipped {nodes [
   n1 = OnWeaponEquipped!() -> (then, NewWeapon:Fort Weapon Object Reference, PrevWeapon:Fort Weapon Object Reference) {is=title K2Node_FunctionEntry}
   > BindToOffhandWeaponEvents(NewOffhandComponent=GetComponentByClass!(self:Actor Object Reference=Get BoundPawn(), ComponentClass:Actor Component Class Reference=/Script/FortniteGame.FortOffhandComponent) -> (ReturnValue:Fort Offhand Component Object Reference))
   > n2 = Cast To FortWeaponRanged_DualWieldable!(execute, Object=@n1) -> (then, CastFailed, AsFort Weapon Ranged Dual Wieldable:Fort Weapon Ranged Dual Wieldable Object Reference) {is=title K2Node_DynamicCast}
   OnUnequipOffhandWeapon(execute=@n2.CastFailed) {is=title K2Node_CallFunction}
  ]}
 SetNoAmmoTextVisibility {nodes [
   n1 = SetNoAmmoTextVisibility!() -> (then, CurrentAmmo:int) {is=title K2Node_FunctionEntry}
   n2 = int - int(A=GetLocalRemainingAmmo(self=Get OffhandWeapon()) -> (ReturnValue=0), B=~)
   > n3 = Branch(Condition=AND Boolean(A="Equal (Integer)"(A=@n2), B="Equal (Integer)"(A=@n1)))
   n4 = Get Text_OutOfAmmo() -> (Text_OutOfAmmo:Common Text Object Reference)
   n5 = GetVisibility!(self:Widget Object Reference=@n4) -> (ReturnValue:ESlateVisibility Enum=Visible)
   > Branch(Condition="Not Equal (Enum)"!(A:ESlateVisibility Enum=@n5, B:ESlateVisibility Enum=Visible) -> (ReturnValue:bool))
   > SetVisibility(self=@n4, InVisibility=Visible)
   n6 = Get HorizontalBox_Ammo() -> (HorizontalBox_Ammo:Horizontal Box Object Reference)
   > SetVisibility(self=@n6, InVisibility=Hidden)
   > Return Node!(execute, NoAmmoTextIsVisible:bool=true)
   n7 = Branch(execute=@n3.else, Condition="Not Equal (Enum)"(A=@n5, B=Hidden))
   > SetVisibility(self=@n4, InVisibility=Hidden)
   > n8 = SetVisibility(self=@n6, InVisibility=Visible)
   Return Node(execute=[@n8.then, @n7.else], NoAmmoTextIsVisible=false)
  ]}
 UpdateTextOpacity {nodes [
   n1 = UpdateTextOpacity!() -> (then, TextWidget:Common Text Object Reference, CurrentValue:int) {is=title K2Node_FunctionEntry}
   n2 = SelectFloat(
    A=Get DefaultAmmoOpacity() -> (DefaultAmmoOpacity:double)
    B=Get EmptyAmmoOpacity() -> (EmptyAmmoOpacity:double)
    bPickA="float > float"(A:int=@n1.CurrentValue)
   )
   > n3 = Branch(Condition="Nearly Equal (Float)"(A=GetRenderOpacity!(self:Widget Object Reference=@n1) -> (ReturnValue:float), B=@n2, ErrorTolerance=0))
   Set RenderOpacity!(execute=@n3.else, RenderOpacity:float=@n2, self:Widget Object Reference=@n1) -> (then, Output_Get:float)
  ]}
]
variables [
 BoundOffhandComponent {type=FortOffhandComponent, default=null}
 BoundPawn {type=FortPlayerPawn}
 DefaultAmmoOpacity {type=number, default=1}
 EmptyAmmoOpacity {type=number, default=0.5}
 InfiniteAmmo {type=integer, default=2000000}
 OffhandWeapon {type=FortWeaponRanged, default=null}
 OwningFortPlayerController {type=FortPlayerController, default=null}
 ZeroAmmo {type=integer}
]
widgets {
 info {inheritedWidgetCount=1, parentClass=&FortHUDElementWidget, rootWidgetClass=&SizeBox, widgetCount=9}
 widgets [
  SizeBox_0 {
   widgetClassPath=&SizeBox
   widget=SizeBox_0
   properties {bOverride_HeightOverride=true, bOverride_WidthOverride=true, heightOverride=55, widthOverride=250}
  }
  Overlay_477 {
   widgetClassPath=&Overlay
   parent=SizeBox_0
   widget=Overlay_477
   slot {class=SizeBoxSlot, properties {horizontalAlignment=HAlign_Right, verticalAlignment=VAlign_Center}}
  }
  HorizontalBox_Ammo {
   widgetClassPath=&HorizontalBox
   bIsVariable=true
   parent=Overlay_477
   widget=HorizontalBox_Ammo
   slot {class=OverlaySlot, properties {horizontalAlignment=HAlign_Right, verticalAlignment=VAlign_Center}}
  }
  Text_CurrentAmmo {
   widgetClassPath=&CommonTextBlock
   bIsVariable=true
   parent=HorizontalBox_Ammo
   widget=Text_CurrentAmmo
   properties {
    text=10
    font {
     bForceMonospaced=false
     bMaterialIsStencil=false
     fontMaterial=null
     fontObject=&Burbank
     letterSpacing=0
     monospacedWidth=1
     size=33
     skewAmount=0
     typefaceFontName=BurbankRegular-Black
     outlineSettings {
      bApplyOutlineToDropShadows=false
      bMiteredCorners=false
      bSeparateFillAlpha=false
      outlineMaterial=null
      outlineSize=0
      outlineColor=<rgba 0 0 0 1>
     }
    }
    shadowOffset=<0 0>
   }
   slot {class=HorizontalBoxSlot, properties {horizontalAlignment=HAlign_Right, verticalAlignment=VAlign_Center, padding=<ltrb 8 0 0 0>}}
  }
  Image_AmmoSplit {
   widgetClassPath=&Image
   bIsVariable=true
   parent=HorizontalBox_Ammo
   widget=Image_AmmoSplit
   properties {brush {
     bIsDynamicallyLoaded=false
     drawAs=Image
     imageType=FullColor
     mirroring=NoMirror
     resourceName=None
     resourceObject=/Game/Athena/HUD/Art/Materials/M_UI_HUD_AmmoCount_DividerLine
     tiling=NoTile
     imageSize=<4 22>
     margin=<ltrb 0 0 0 0>
     outlineSettings {
      bUseBrushTransparency=false
      roundingType=HalfHeightRadius
      width=0
      color {colorUseRule=UseColor_Specified, specifiedColor=<rgba 0 0 0 0>}
      cornerRadii=<0 0 0 0>
     }
     tintColor {colorUseRule=UseColor_Specified, specifiedColor=<rgba 1 1 1 1>}
     uVRegion {bIsValid=false, max=<0 0>, min=<0 0>}
    }}
   slot {class=HorizontalBoxSlot, properties {verticalAlignment=VAlign_Center, padding=<ltrb 8 0 8 0>}}
  }
  Text_TotalAmmo {
   widgetClassPath=&CommonTextBlock
   bIsVariable=true
   parent=HorizontalBox_Ammo
   widget=Text_TotalAmmo
   properties {
    text=10
    font {
     bForceMonospaced=false
     bMaterialIsStencil=false
     fontMaterial=null
     fontObject=&Burbank
     letterSpacing=0
     monospacedWidth=1
     size=16.5
     skewAmount=0
     typefaceFontName=Burbank-PlayerPower
     outlineSettings {
      bApplyOutlineToDropShadows=false
      bMiteredCorners=false
      bSeparateFillAlpha=false
      outlineMaterial=null
      outlineSize=0
      outlineColor=<rgba 0 0 0 1>
     }
    }
    shadowOffset=<0 0>
   }
   slot {class=HorizontalBoxSlot, properties {horizontalAlignment=HAlign_Left, verticalAlignment=VAlign_Center}}
  }
  Image_AmmoType {
   widgetClassPath=&Image
   bIsVariable=true
   parent=HorizontalBox_Ammo
   widget=Image_AmmoType
   properties {brush {
     bIsDynamicallyLoaded=false
     drawAs=Image
     imageType=FullColor
     mirroring=NoMirror
     resourceName=None
     resourceObject=/Game/UI/Foundation/Textures/Icons/Ammo/T-Icon-Ammo-Medium-64
     tiling=NoTile
     imageSize=<24 24>
     margin=<ltrb 0 0 0 0>
     outlineSettings {
      bUseBrushTransparency=false
      roundingType=HalfHeightRadius
      width=0
      color {colorUseRule=UseColor_Specified, specifiedColor=<rgba 0 0 0 0>}
      cornerRadii=<0 0 0 0>
     }
     tintColor {colorUseRule=UseColor_Specified, specifiedColor=<rgba 1 1 1 1>}
     uVRegion {bIsValid=false, max=<0 0>, min=<0 0>}
    }}
   slot {class=HorizontalBoxSlot, properties {horizontalAlignment=HAlign_Left, verticalAlignment=VAlign_Center, padding=<ltrb 4 0 0 0>}}
  }
  Text_OutOfAmmo {
   widgetClassPath=&CommonTextBlock
   bIsVariable=true
   parent=Overlay_477
   widget=Text_OutOfAmmo
   properties {
    mobileFontSizeMultiplier=1.3
    text=NO AMMO
    visibility=Hidden
    colorAndOpacity {colorUseRule=UseColor_Specified, specifiedColor=<rgba 1 0.0168 0.0953 1>}
    font {
     bForceMonospaced=false
     bMaterialIsStencil=false
     fontMaterial=null
     fontObject=&Burbank
     letterSpacing=0
     monospacedWidth=1
     size=18
     skewAmount=0
     typefaceFontName=BurbankRegular-Black
     outlineSettings {
      bApplyOutlineToDropShadows=false
      bMiteredCorners=false
      bSeparateFillAlpha=true
      outlineMaterial=null
      outlineSize=1
      outlineColor=<rgba 1 0.0168 0.0953 0.05>
     }
    }
    shadowOffset=<0 0>
   }
   slot {class=OverlaySlot, properties {horizontalAlignment=HAlign_Right, verticalAlignment=VAlign_Center}}
  }
  TutorialHighlight {widgetClassPath=&TutorialHighlight}
 ]
}
