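# Translucent UI material that draws two radial segmented ammo bars, SMG and Shotgun, for a dual-weapon reticle.
# Each weapon's active ammo and clip size feed a separate MF_UI_Reticle_Bar_Radial_Ammo call, with a different circle size for the Shotgun.
# Reload flags switch each bar's fill to follow its reload timer and cap brightness with a sine-based pulse.
# Max merges both bars' values and masks, with MasterOpacity scaling the combined opacity.
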Material /BattleRoyaleDualWeaponUI/Art/Materials/M_UI_Reticle_DualAmmoBars
graph {
 properties {bAutomaticallySetUsageInEditor=false, blendMode=BLEND_Translucent, materialDomain=MD_UI, usedShadingModels=Default Lit}
 nodes [
  Constant() {desc=Steps for Interp, r=1000}
  Constant() {desc=Steps for Interp, r=1000}
  n1 = Abs(None=Sine(None=Time()))
  n2 = Clamp(None=Multiply(A=Power(Base=@n1) {constExponent=1})) {minDefault=0.5}
  n3 = ScalarParameter() {defaultValue=20, parameterName=SMG Clip Size}
  n4 = ScalarParameter() {defaultValue=0.5, parameterName=SMG Reload Timer}
  n5 = Ceil(None=ScalarParameter() {defaultValue=1, parameterName=isSMGReloading})
  n6 = LinearInterpolate(
   A=AppendVector(A=ScalarParameter() {defaultValue=10, parameterName=SMG Active Ammo}, B=@n3)
   B=AppendVector(A=LinearInterpolate(B=@n3, Alpha=@n4) {constB=20}, B=@n3)
   Alpha=@n5
  ) {constA=0.1, constB=0}
  n7 = Constant() {r=0.4}
  n8 = MaterialFunctionCall!(
   UVs
   Clip Size=ComponentMask(None=@n6) {g=true}
   Active Ammo=ComponentMask(None=@n6) {r=true}
   Reload Progress
   Segment Spacing=LinearInterpolate(Alpha=@n5) {constA=0.2, constB=0.2}
   "Start Angle (0-1)"
   "End Angle (0-1)"
   "Background Opacity (0-1)"=@n7
   "Circle Size (0-1)"
   Outline Thickness
   Fixed Spacing
  ) {materialFunction=/FortUILibrary/Systems/Reticles/MaterialFunctions/AuxillaryProgressBars/MF_UI_Reticle_Bar_Radial_Ammo}
  n9 = LinearInterpolate(
   B=LinearInterpolate(Alpha=Multiply(A=@n8.Opacity Mask, B=@n2)) {constB=0.7}
   Alpha=Ceil(None=ScalarParameter() {defaultValue=1, parameterName=isSMGReloading})
  ) {constA=1}
  n10 = ScalarParameter() {defaultValue=8, parameterName=Shotgun Clip Size}
  n11 = ScalarParameter() {defaultValue=0.5, parameterName=Shotgun Reload Timer}
  n12 = Ceil(None=ScalarParameter() {defaultValue=1, parameterName=isShotgunReloading})
  n13 = LinearInterpolate(
   A=AppendVector(A=ScalarParameter() {defaultValue=4, parameterName=Shotgun Active Ammo}, B=@n10)
   B=AppendVector(A=LinearInterpolate(B=@n10, Alpha=@n11) {constB=20}, B=@n10)
   Alpha=@n12
  )
  n14 = MaterialFunctionCall(
   Clip Size=ComponentMask(None=@n13) {g=true}
   Active Ammo=ComponentMask(None=@n13) {r=true}
   Segment Spacing=LinearInterpolate(Alpha=@n12) {constA=0.1, constB=0.1}
   "Background Opacity (0-1)"=@n7
   "Circle Size (0-1)"=Constant() {r=0.575}
  ) {materialFunction=&MF_UI_Reticle_Bar_Radial_Ammo}
  n15 = LinearInterpolate(
   B=LinearInterpolate(Alpha=Multiply(A=@n14.Opacity Mask, B=@n2)) {constB=0.7}
   Alpha=Ceil(None=ScalarParameter() {defaultValue=1, parameterName=isShotgunReloading})
  ) {constA=1}
  Max(
   A=LinearInterpolate(B=Clamp(None=@n8.Progress Mask, Max=@n9), Alpha=@n8.Opacity Mask)
   B=LinearInterpolate(B=Clamp(None=@n14.Progress Mask, Max=@n15), Alpha=@n14.Opacity Mask)
  )
  Multiply(A=Max(A=@n8.Opacity Mask, B=@n14.Opacity Mask), B=ScalarParameter() {defaultValue=0.65, parameterName=MasterOpacity}) {constB=0.65}
 ]
}
