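# Base reticle extension widget for dual weapons, showing ammo and reload bars for a primary (shotgun) and offhand (SMG) weapon.
# On view target changes it rebinds ammo and reload events on the equipped weapon and the offhand component's weapon.
# Events drive one dynamic material via scalar parameters and animations time-scaled to reload time; primary reloads also trigger the SMG animation during dual reloads.
# A restarted timer fades the bars out.
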WidgetBlueprint /BattleRoyaleDualWeaponUI/Widgets/WBP_Reticle_DualWeaponBase
animations [
 HideAmmoBars {
  binding_count=1
  length_seconds=0.5
  object=&self:HideAmmoBars
  sequence {bindings [AmmoBars {
     id=#1
     possessed_class=Image
     tracks [{
       type=MovieSceneFloatTrack
       display_name=Render Opacity
       sections [{type=MovieSceneFloatSection, end_frame=10, start_frame=0, channels [None {type=MovieSceneScriptingFloatChannel, keys [[0,1],[10,0]]}]}]
      }]
    }]}
 }
 SecondaryReloading {
  binding_count=1
  length_seconds=1.0001
  object=&self:SecondaryReloading
  sequence {
   bindings [AmmoBars {
     id=#2
     possessed_class=Image
     tracks [{
       type=MovieSceneWidgetMaterialTrack
       sections [{
         type=MovieSceneParameterSection
         end_frame=30
         start_frame=0
         channels [
          isSMGReloading {type=MovieSceneScriptingFloatChannel, keys [[0,1],[30,0]]}
          SMG Reload Timer {type=MovieSceneScriptingFloatChannel, keys [[0,0],[27,1]]}
         ]
        }]
      }]
    }]
   display_rate {numerator=30}
  }
 }
 PrimaryReloading {
  binding_count=1
  length_seconds=1.0051
  object=&self:PrimaryReloading
  sequence {
   bindings [AmmoBars {
     id=#3
     possessed_class=Image
     tracks [{
       type=MovieSceneWidgetMaterialTrack
       sections [{
         type=MovieSceneParameterSection
         end_frame=30
         start_frame=0
         channels [
          isShotgunReloading {type=MovieSceneScriptingFloatChannel, keys [[0,1],[30,0]]}
          Shotgun Reload Timer {type=MovieSceneScriptingFloatChannel, keys [[0,0],[27,1]]}
         ]
        }]
      }]
    }]
   display_rate {numerator=30}
  }
 }
]
graphs [
 BindOffhandWeaponEvents {nodes [
   n1 = BindOffhandWeaponEvents!() -> (then, NewWeapon:Fort Weapon Object Reference) {is=title K2Node_FunctionEntry}
   > n2 = Cast To FortWeaponRanged!(execute, Object=@n1) -> (then, CastFailed, AsFort Weapon Ranged:Fort Weapon Ranged Object Reference) {is=title K2Node_DynamicCast}
   > n3 = Set varOffhandWeapon(varOffhandWeapon:Fort Weapon Ranged Object Reference=@n2) -> (Output_Get:Fort Weapon Ranged Object Reference)
   > Is Valid(InputObject=@n3)
   > Bind Event to On Local Ammo Changed(self:Fort Weapon Object Reference=@n3, Delegate=Create Event()) {is=title K2Node_AddDelegate}
   > Bind Event to On Local Reload Started(self:Fort Weapon Object Reference=@n3, Delegate=Create Event()) {is=title K2Node_AddDelegate}
   > Bind Event to On Local Reload Cancelled(self:Fort Weapon Object Reference=@n3, Delegate=Create Event()) {is=title K2Node_AddDelegate}
  ]}
 BindPrimaryWeaponEvents {nodes [
   BindPrimaryWeaponEvents() {is=title K2Node_FunctionEntry}
   > n1 = Cast To FortWeaponRanged(Object=GetEquippedWeapon!(self:Fort Weapon Reticle Extension Widget Base Object Reference) -> (ReturnValue:Fort Weapon Object Reference))
   > n2 = Set varPrimaryWeapon(varPrimaryWeapon:Fort Weapon Ranged Object Reference=@n1) -> (Output_Get:Fort Weapon Ranged Object Reference)
   > Bind Event to On Local Ammo Changed(self=@n2, Delegate=Create Event())
   > Bind Event to On Local Reload Started(self=@n2, Delegate=Create Event())
   > Bind Event to On Local Reload Cancelled(self=@n2, Delegate=Create Event())
   > n3 = Cast To FortWeaponRanged_DualWieldable!(execute, Object=@n2) -> (then, CastFailed, AsFort Weapon Ranged Dual Wieldable:Fort Weapon Ranged Dual Wieldable Object Reference) {is=title K2Node_DynamicCast}
   > Set DualWieldable(DualWieldable:Fort Weapon Ranged Dual Wieldable Object Reference=@n3) -> (Output_Get:Fort Weapon Ranged Dual Wieldable Object Reference)
  ]}
 EventGraph {nodes [
   n1 = ShotgunAmmoChanged!() -> (OutputDelegate, then, LocalCount:int, LocalRemaining:int) {is=title K2Node_CustomEvent}
   > InterruptFadeOut()
   > SetScalarParameterValue(
    self=Get varDynamicAmmo() -> (varDynamicAmmo:Material Instance Dynamic Object Reference)
    ParameterName=Shotgun Active Ammo
    Value="To Float (Integer)"(InInt=@n1.LocalCount)
   )
   > Clear and Invalidate Timer by Handle(Handle=Get AutoHideTime() -> (AutoHideTime:Timer Handle Structure))
   > HideAmmoBarsTimer()
   n2 = onViewTargetChanged!() -> (
    OutputDelegate
    then
    PlayerController:Fort Player Controller Object Reference
    OldViewTarget:Actor Object Reference
    NewViewTarget:Actor Object Reference
   ) {is=title K2Node_CustomEvent}
   > n3 = Sequence()
   > UnbindPrimaryWeaponEvents()
   > BindPrimaryWeaponEvents() {is=title K2Node_CallFunction}
   n4 = Set varOffhandComponent(
    execute=@n3.then_1
    varOffhandComponent:Fort Offhand Component Object Reference=GetComponentByClass!(self:Actor Object Reference=@n2.NewViewTarget, ComponentClass:Actor Component Class Reference=/Script/FortniteGame.FortOffhandComponent) -> (ReturnValue:Fort Offhand Component Object Reference)
   ) -> (Output_Get:Fort Offhand Component Object Reference)
   > n5 = Is Valid(InputObject=@n4)
   UnbindOffhandWeaponEvents(execute=@n5.Is Not Valid)
   n6 = GetCurrentEquippedOffhandWeapon!(
    execute=@n5.Is Valid
    self:Fort Offhand Component Object Reference=Get varOffhandComponent() -> (varOffhandComponent:Fort Offhand Component Object Reference)
   ) -> (then, ReturnValue:Fort Weapon Object Reference)
   > UnbindOffhandWeaponEvents()
   > BindOffhandWeaponEvents!(execute, self, NewWeapon:Fort Weapon Object Reference=@n6) -> (then) {is=title K2Node_CallFunction}
   > IntiailizeAmmoBars()
   > ResetMaterial()
   > HideAmmoBarsTimer()
   OnSMGReloadCancelled() {is=title K2Node_CustomEvent}
   > StopReloadAnimation!(
    execute
    self
    InAnimation:Widget Animation Object Reference=Get SecondaryReloading() -> (SecondaryReloading:Widget Animation Object Reference)
   ) -> (then)
   n7 = SMGAmmoChange!() -> (OutputDelegate, then, LocalCount:int, LocalRemaining:int) {is=title K2Node_CustomEvent}
   > InterruptFadeOut()
   > SetScalarParameterValue(self=Get varDynamicAmmo(), ParameterName=SMG Active Ammo, Value="To Float (Integer)"(InInt=@n7.LocalCount))
   > Clear and Invalidate Timer by Handle(Handle=Get AutoHideTime())
   > HideAmmoBarsTimer()
   n8 = HideAmmoBarsTimer() {is=title K2Node_CustomEvent}
   n9 = HideBars() {is=title K2Node_CustomEvent}
   n10 = Set Timer by Event(execute=@n8.then, Delegate=@n9, Time=2, bLooping=false)
   > Set AutoHideTime(AutoHideTime:Timer Handle Structure=@n10) -> (Output_Get:Timer Handle Structure)
   n11 = OnShotgunReloadStarted!() -> (OutputDelegate, then, ReloadTime:float=~, ReloadType:EFortWeaponReloadType Enum=ReloadWholeClip) {is=title K2Node_CustomEvent}
   > PlayReloadAnimation!(
    execute
    self
    Reload Animation:Widget Animation Object Reference=Get PrimaryReloading() -> (PrimaryReloading:Widget Animation Object Reference)
    Reaload Time:double=@n11.ReloadTime
   ) -> (then)
   > Clear and Invalidate Timer by Handle(Handle=Get AutoHideTime())
   > HideAmmoBarsCustomTime!(execute, self, Delay time:double=@n11.ReloadTime) -> (then)
   n12 = Get DualWieldable() -> (DualWieldable:Fort Weapon Ranged Dual Wieldable Object Reference)
   n13 = GetReloadTime!(self:Fort Weapon Object Reference=Get varPrimaryWeapon() -> (varPrimaryWeapon:Fort Weapon Ranged Object Reference)) -> (ReturnValue:float)
   > Branch(Condition=AND Boolean(A=GetIsTryingDualReload!(self:Fort Weapon Ranged Dual Wieldable Object Reference=@n12) -> (ReturnValue:bool), B="float > float"(A=@n13)))
   > OnSMGReloadStarted!(execute, self, ReloadTime:float=@n13, ReloadType:EFortWeaponReloadType Enum=ReloadWholeClip) -> (then)
   n14 = OnSMGReloadStarted!() -> (OutputDelegate, then, ReloadTime:float=~, ReloadType:EFortWeaponReloadType Enum=ReloadWholeClip) {is=title K2Node_CustomEvent}
   > PlayReloadAnimation(Reload Animation=Get SecondaryReloading(), Reaload Time=@n14.ReloadTime)
   > Clear and Invalidate Timer by Handle(Handle=Get AutoHideTime())
   > HideAmmoBarsCustomTime(Delay time=@n14.ReloadTime)
   n15 = HideAmmoBarsCustomTime!() -> (OutputDelegate, then, Delay time:double=~) {is=title K2Node_CustomEvent}
   > n16 = Set Timer by Event(Delegate=@n9, Time=@n15.Delay time, bLooping=false)
   > Set AutoHideTime(AutoHideTime=@n16)
   OnShotgunReloadCancelled() {is=title K2Node_CustomEvent}
   > StopReloadAnimation(InAnimation=Get PrimaryReloading())
   > Branch(Condition=GetIsTryingDualReload(self=Get DualWieldable()))
   > OnSMGReloadCancelled() {is=title K2Node_CallFunction}
   Event Construct()
   n17 = GetOwningPlayer()
   > n18 = Cast To FortPlayerController(Object=@n17)
   > n19 = Set owningFortPlayerController(owningFortPlayerController:Fort Player Controller Object Reference=@n18) -> (Output_Get:Fort Player Controller Object Reference)
   > Bind Event to On View Target Changed(self:Fort Player Controller Object Reference=@n19, Delegate=Create Event()) {is=title K2Node_AddDelegate}
   > onViewTargetChanged!(
    execute
    self
    PlayerController:Fort Player Controller Object Reference
    OldViewTarget:Actor Object Reference
    NewViewTarget:Actor Object Reference=GetViewTarget!(self:Controller Object Reference=@n17) -> (ReturnValue:Actor Object Reference)
   ) -> (then) {is=title K2Node_CallFunction}
   Event OnInitialized()
   > n20 = GetDynamicMaterial(self=Get AmmoBars() -> (AmmoBars:Image Object Reference))
   > Set varDynamicAmmo(varDynamicAmmo:Material Instance Dynamic Object Reference=@n20) -> (Output_Get:Material Instance Dynamic Object Reference)
   Event Destruct()
   > n21 = Sequence()
   n22 = Get owningFortPlayerController() -> (owningFortPlayerController:Fort Player Controller Object Reference)
   > Is Valid(InputObject=@n22)
   > Unbind Event from On View Target Changed(self:Fort Player Controller Object Reference=@n22, Delegate=Create Event()) {is=title K2Node_RemoveDelegate}
   > Set owningFortPlayerController()
   UnbindPrimaryWeaponEvents(execute=@n21.then_1)
   > UnbindOffhandWeaponEvents()
   PlayAnimationForward(execute=@n9.then, InAnimation=Get HideAmmoBars() -> (HideAmmoBars:Widget Animation Object Reference))
  ]}
 InterruptFadeOut {nodes [
   InterruptFadeOut() {is=title K2Node_FunctionEntry}
   n1 = Get HideAmmoBars()
   > n2 = Branch(Condition=IsAnimationPlaying(InAnimation=@n1))
   > n3 = StopAnimation(InAnimation=@n1)
   Set RenderOpacity!(execute=[@n3.then, @n2.else], RenderOpacity:float=1, self:Widget Object Reference=Get AmmoBars()) -> (then, Output_Get:float)
  ]}
 IntiailizeAmmoBars {nodes [
   IntiailizeAmmoBars() {is=title K2Node_FunctionEntry}
   n1 = GetEquippedWeapon()
   n2 = Get varDynamicAmmo()
   > SetScalarParameterValue(
    self=@n2
    ParameterName=Shotgun Clip Size
    Value="To Float (Integer)"(InInt=GetBulletsPerClip!(self:Fort Weapon Object Reference=@n1) -> (ReturnValue:int=0))
   )
   > SetScalarParameterValue(
    self=@n2
    ParameterName=Shotgun Active Ammo
    Value="To Float (Integer)"(InInt=GetLocalMagazineAmmoCount!(self:Fort Weapon Object Reference=@n1) -> (ReturnValue:int=0))
   )
   n3 = Get varOffhandWeapon() -> (varOffhandWeapon:Fort Weapon Ranged Object Reference)
   n4 = Get varDynamicAmmo()
   > SetScalarParameterValue(self=@n4, ParameterName=SMG Clip Size, Value="To Float (Integer)"(InInt=GetBulletsPerClip(self=@n3) -> (ReturnValue=0)))
   > SetScalarParameterValue(
    self=@n4
    ParameterName=SMG Active Ammo
    Value="To Float (Integer)"(InInt=GetMagazineAmmoCount!(self:Fort Weapon Object Reference=@n3) -> (ReturnValue:int=0))
   )
  ]}
 PlayReloadAnimation {nodes [
   n1 = PlayReloadAnimation!() -> (then, Reload Animation:Widget Animation Object Reference, Reaload Time:double=~) {is=title K2Node_FunctionEntry}
   > InterruptFadeOut() {is=title K2Node_CallFunction}
   n2 = "float > float"(A=@n1.Reaload Time, B=0) {is=title K2Node_CallFunction}
   > PlayAnimation(
    InAnimation=@n1
    PlaybackSpeed=float / float(A=1, B=Select(Option 0:double=1, Option 1:double=@n1.Reaload Time, Index=@n2) -> (ReturnValue:double)) {is=title K2Node_CallFunction}
   )
  ]}
 ResetMaterial {nodes [
   ResetMaterial() {is=title K2Node_FunctionEntry}
   n1 = Get varDynamicAmmo()
   > SetScalarParameterValue(self=@n1, ParameterName=isShotgunReloading, Value=0)
   > SetScalarParameterValue(self=@n1, ParameterName=isSMGReloading, Value=0)
   > Set RenderOpacity(RenderOpacity=1, self=Get AmmoBars())
  ]}
 StopReloadAnimation {nodes [
   n1 = StopReloadAnimation!() -> (then, InAnimation:Widget Animation Object Reference) {is=title K2Node_FunctionEntry}
   > Branch(Condition=IsAnimationPlaying(InAnimation=@n1))
   > PlayAnimation(InAnimation=@n1, StartAtTime=GetEndTime(self=@n1))
   > ResetMaterial() {is=title K2Node_CallFunction}
  ]}
 UnbindOffhandWeaponEvents {nodes [
   UnbindOffhandWeaponEvents() {is=title K2Node_FunctionEntry}
   > Is Valid(InputObject=Get varOffhandWeapon())
   n1 = Get varOffhandWeapon()
   > Unbind Event from On Local Reload Started(self:Fort Weapon Object Reference=@n1, Delegate=Create Event()) {is=title K2Node_RemoveDelegate}
   > Unbind Event from On Local Reload Cancelled(self:Fort Weapon Object Reference=@n1, Delegate=Create Event()) {is=title K2Node_RemoveDelegate}
   > Unbind Event from On Local Ammo Changed(self:Fort Weapon Object Reference=@n1, Delegate=Create Event()) {is=title K2Node_RemoveDelegate}
   > Set varOffhandWeapon()
  ]}
 UnbindPrimaryWeaponEvents {nodes [
   UnbindPrimaryWeaponEvents() {is=title K2Node_FunctionEntry}
   n1 = Get varPrimaryWeapon()
   > Is Valid(InputObject=@n1)
   > Unbind Event from On Local Ammo Changed(self=@n1, Delegate=Create Event())
   > Unbind Event from On Local Reload Started(self=@n1, Delegate=Create Event())
   > Unbind Event from On Local Reload Cancelled(self=@n1, Delegate=Create Event())
   > Set varPrimaryWeapon()
   > Set DualWieldable()
  ]}
]
variables [
 AutoHideTime {type=TimerHandle}
 DualWieldable {type=FortWeaponRanged_DualWieldable, default=null}
 owningFortPlayerController {type=FortPlayerController, default=null}
 varDynamicAmmo {type=MaterialInstanceDynamic}
 varOffhandComponent {type=FortOffhandComponent, default=null}
 varOffhandWeapon {type=FortWeaponRanged, default=null}
 varPrimaryWeapon {type=FortWeaponRanged, default=null}
]
widgets {
 info {
  inheritedWidgetCount=1
  parentClass=/Script/FortniteUI.FortWeaponReticleExtensionWidgetBase
  rootWidgetClass=&Overlay
  widgetCount=4
 }
 widgets [
  Overlay_39 {widgetClassPath=&Overlay, widget=Overlay_39}
  NamedSlot_Center {
   widgetClassPath=&NamedSlot
   parent=Overlay_39
   widget=NamedSlot_Center
   slot {class=OverlaySlot, properties {horizontalAlignment=HAlign_Center, verticalAlignment=VAlign_Center}}
  }
  AmmoBars {
   widgetClassPath=&Image
   bIsVariable=true
   parent=Overlay_39
   widget=AmmoBars
   properties {
    brush {
     bIsDynamicallyLoaded=false
     drawAs=Image
     imageType=FullColor
     mirroring=NoMirror
     resourceName=None
     resourceObject=/BattleRoyaleDualWeaponUI/Art/Materials/M_UI_Reticle_DualAmmoBars
     tiling=NoTile
     imageSize=<256 256>
     margin=<ltrb 0 0 0 0>
     outlineSettings {
      bUseBrushTransparency=false
      roundingType=HalfHeightRadius
      width=0
      color {colorUseRule=UseColor_Specified, specifiedColor=<rgba 0 0 0 0>}
      cornerRadii=<0 0 0 0>
     }
     tintColor {colorUseRule=UseColor_Specified, specifiedColor=<rgba 1 1 1 1>}
     uVRegion {bIsValid=false, max=<0 0>, min=<0 0>}
    }
    renderTransform {angle=0, scale=<1 1>, shear=<0 0>, translation=<30 30>}
   }
   slot {class=OverlaySlot, properties {horizontalAlignment=HAlign_Center, verticalAlignment=VAlign_Center}}
  }
  TutorialHighlight {widgetClassPath=&TutorialHighlight}
 ]
}
